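# Sound cue for distant metal water pump squeaks, a spot ambient one-shot at the Knights Village POI in Asteria.
# A Random node picks one of seven equally weighted squeak waves, feeding a Quality Level switch that fills only its first two slots, then a Modulator at the root.
# It plays slightly reduced in volume through the ambient one-shots sound class, borrowing a tile-ambient bird attenuation, with submix sends disabled.
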SoundCue /Game/Sounds/Asteria/POI/Knights_Village/Spots/Spot_Knights_Village_Metal_Waterpump_Squeaks_Distant_Cue
root=Modulator_0
nodes [
 {id=Modulator_0, children [QualityLevel_0]}
 {id=QualityLevel_0, children [Random_0,Random_0,null,null]}
 {
  id=Random_0
  children [WavePlayer_0,WavePlayer_1,WavePlayer_2,WavePlayer_3,WavePlayer_4,WavePlayer_5,WavePlayer_6]
  properties {weights [1,1,1,1,1,1,1]}
 }
 {
  id=WavePlayer_0
  properties {soundWaveAssetPtr=&/Spot_Knights_Village_Metal_Waterpump_Squeak_01}
 }
 {
  id=WavePlayer_1
  properties {soundWaveAssetPtr=&/Spot_Knights_Village_Metal_Waterpump_Squeak_02}
 }
 {
  id=WavePlayer_2
  properties {soundWaveAssetPtr=&/Spot_Knights_Village_Metal_Waterpump_Squeak_03}
 }
 {
  id=WavePlayer_3
  properties {soundWaveAssetPtr=&/Spot_Knights_Village_Metal_Waterpump_Squeak_04}
 }
 {
  id=WavePlayer_4
  properties {soundWaveAssetPtr=&/Spot_Knights_Village_Metal_Waterpump_Squeak_05}
 }
 {
  id=WavePlayer_5
  properties {soundWaveAssetPtr=&/Spot_Knights_Village_Metal_Waterpump_Squeak_06}
 }
 {
  id=WavePlayer_6
  properties {soundWaveAssetPtr=&/Spot_Knights_Village_Metal_Waterpump_Squeak_07}
 }
]
properties {
 attenuationSettings=&Fort_TileAmbient_OneShot_Bird_Att
 bEnableSubmixSends=false
 maxDistance=8005
 soundClassObject=&Fort_Ambient_OneShots
 volumeMultiplier=0.752
 preEffectBusSends [{customSendLevelCurve {editorCurveData {keys [<at 0 0>,<at 1 1>]}}}]
 soundSubmixSends [{minSendLevel=0, sendLevel=0, soundSubmix=null}]
}
