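# Knights Village ambient sound cue for gentle water lapping against a wooden pontoon, mixed from three separately attenuated layers.
# An endless loop plays random wood impact one-shots with varied delay and pitch, using occlusion-named attenuation.
# Pontoon slapping and quieter-weighted lake shore loops each run through repeating Enveloper volume curves, so they swell and dip over time.
# Concurrency override allows one instance, preventing new ones.
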SoundCue /Game/Sounds/Asteria/POI/Knights_Village/Spots/Spot_Water_Wood_Pontoon_Gentle_Lapping_Occ_Cue
root=Mixer_0
nodes [
 {
  id=Attenuation_0
  children [Enveloper_0]
  properties {attenuationSettings=/Game/Athena/Sounds/Water/Attenuations/Apollo_Water_River_Close_2D_Att}
 }
 {
  id=Attenuation_5
  children [Looping_0]
  properties {attenuationSettings=/Game/Athena/Sounds/Water/Attenuations/Apollo_Water_Slapping_Impact_Foley_Focus_Occ_Att}
 }
 {
  id=Attenuation_6
  children [Enveloper_1]
  properties {attenuationSettings=/Game/Athena/Sounds/Water/Attenuations/Apollo_Water_River_Close_Focused_Att}
 }
 {id=Delay_0, children [Modulator_0], properties {delayMax=8, delayMin=3}}
 {
  id=Enveloper_0
  children [WavePlayer_18]
  properties {
   bLoop=true
   bLoopIndefinitely=true
   loopEnd=5
   loopStart=0.1
   pitchCurve {editorCurveData {keys [{time=0}]}}
   volumeCurve {editorCurveData {keys [
      <at 0.1 1 RCIM_Cubic>
      <at 0.5 4 RCIM_Cubic>
      <at 1.5 1>
      {arriveTangent=9, interpMode=RCIM_Cubic, leaveTangent=9, time=2, value=1}
      <at 2.5 10 RCIM_Cubic>
      {arriveTangent=0.011, interpMode=RCIM_Cubic, leaveTangent=0.011, tangentMode=RCTM_User, time=4, value=1}
      <at 5 1>
     ]}}
  }
 }
 {
  id=Enveloper_1
  children [WavePlayer_0]
  properties {
   bLoop=true
   bLoopIndefinitely=true
   loopEnd=10
   pitchCurve {editorCurveData {keys [{time=0}]}}
   volumeCurve {editorCurveData {keys [
      <at 0 1 RCIM_Cubic>
      <at 5 0.25 RCIM_Cubic>
      {arriveTangent=0.011, interpMode=RCIM_Cubic, leaveTangent=0.011, tangentMode=RCTM_User, time=10, value=1}
     ]}}
  }
 }
 {id=Looping_0, children [Delay_0]}
 {id=Mixer_0, children [Attenuation_5, Attenuation_6, Attenuation_0], properties {inputVolume [0.4,0.8,0.2]}}
 {id=Modulator_0, children [Random_0], properties {pitchMax=1.1, pitchMin=0.9}}
 {id=Random_0, children [WavePlayer_3,WavePlayer_4,WavePlayer_5,WavePlayer_6,WavePlayer_7,WavePlayer_8], properties {weights [1,1,1,1,1,1]}}
 {
  id=WavePlayer_0
  properties {bLooping=true, soundWaveAssetPtr=/Game/Sounds/Asteria/POI/IndustrialTown/Spots/Spot_Water_Slapping_Against_Pontoon_Loop}
 }
 {id=WavePlayer_18, properties {bLooping=true, soundWaveAssetPtr=/Game/Athena/Sounds/Water/Ambient/Amb_Water_Lake_Shore_Laps_Loop_01}}
 {id=WavePlayer_3, properties {soundWaveAssetPtr=/Game/Athena/Sounds/Water/Ambient/Spot_Water_Pontoon_Wood_Impact_01}}
 {id=WavePlayer_4, properties {soundWaveAssetPtr=/Game/Athena/Sounds/Water/Ambient/Spot_Water_Pontoon_Wood_Impact_02}}
 {id=WavePlayer_5, properties {soundWaveAssetPtr=/Game/Athena/Sounds/Water/Ambient/Spot_Water_Pontoon_Wood_Impact_03}}
 {id=WavePlayer_6, properties {soundWaveAssetPtr=/Game/Athena/Sounds/Water/Ambient/Spot_Water_Pontoon_Wood_Impact_04}}
 {id=WavePlayer_7, properties {soundWaveAssetPtr=/Game/Athena/Sounds/Water/Ambient/Spot_Water_Pontoon_Wood_Impact_05}}
 {id=WavePlayer_8, properties {soundWaveAssetPtr=/Game/Athena/Sounds/Water/Ambient/Spot_Water_Pontoon_Wood_Impact_06}}
]
properties {
 bOverrideConcurrency=true
 maxDistance=4512
 soundClassObject=/Game/Sounds/FortSoundClassesAndModes/Fort_Water_Ambient
 volumeMultiplier=1.3
 attenuationOverrides {
  bAttenuateWithLPF=true
  falloffDistance=6000
  lPFFrequencyAtMax=500
  nonSpatializedRadiusStart=700
  stereoSpread=1000
  attenuationShapeExtents {x=925}
 }
 concurrencyOverrides {maxCount=1, resolutionRule=PreventNew}
}
